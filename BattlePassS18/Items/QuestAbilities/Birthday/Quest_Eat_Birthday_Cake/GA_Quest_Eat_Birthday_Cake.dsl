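; Gameplay ability for the birthday quest to eat cake, reporting quest progress when triggered by a gameplay event.
; An unblocked activation sets a blocked flag, waits briefly, clears it, then sends a complex custom stat event with quest tags and the event's target tags to the owning player's quest manager.
; Activations arriving while blocked only wait and end, debouncing repeat triggers.
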(event Ability|EventActivateAbilityFromEvent
  (params (EventData :type Gameplay%20Event%20Data%20Structure%20%28by%20ref%29))
  (bind (_eventtag _instigator _target _optionalobject _optionalobject2 _contexthandle _instigatortags _targettags) (Utilities|Struct|BreakGameplayEventData EventData))
  (|CallParentFunction EventData)
  (Development|PrintString "@@@ ActivateAbilityFromEvent [GA_Quest_Eat_Birthday_Cake]: CALLED!" false (Variables|Debug|GetbDebugAbility))
  (if (Variables|Default|GetbBlocked)
    (Utilities|FlowControl|Delay 0.2)
    (Ability|EndAbility)
    (else
      (Variables|Default|SetbBlocked true)
      (Utilities|FlowControl|Delay 0.2)
      (Variables|Default|SetbBlocked false)
      (Quest|SendComplexCustomStatEvent (Variables|GAVars|Data|GetOwningPlayerQuestManager) 0 (Variables|Default|GetQuestTags) _targettags 1)
      (Ability|EndAbility))))
